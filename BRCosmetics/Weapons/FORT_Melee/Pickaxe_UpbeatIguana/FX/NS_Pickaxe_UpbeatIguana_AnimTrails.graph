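# Niagara system for the UpbeatIguana pickaxe's animation trails, drawing four short-lived ribbons from skeletal mesh bones or sockets.
# A soft dark trail, primary and secondary trails and a tip trail are spawned and updated by FN anim-trail modules, with position offsets along ribbon facing on all but the tip.
# User accent colors tint three trails, with special-case remapping of one color.
# Curves and system FadeOut drive alpha; time-of-day colors dim primary and tip at night.
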NiagaraSystem /BRCosmetics/Weapons/FORT_Melee/Pickaxe_UpbeatIguana/FX/NS_Pickaxe_UpbeatIguana_AnimTrails
dependencies {
 usedDataInterfaces [&NiagaraDataInterfaceColorCurve, &NiagaraDataInterfaceSkeletalMesh, &NiagaraDataInterfaceCurve]
 usedDynamicInputs [&FloatFromCurve]
 usedModules [
  &self:SystemSpawnScript.SystemScriptSource.SystemScriptGraph.NiagaraNodeAssignment_0.SetVariables_#1
  &SystemState
  &self:SystemSpawnScript.SystemScriptSource.SystemScriptGraph.NiagaraNodeAssignment_1.SetVariables_#2
 ]
}
emitters {
 NE_SoftDarkTrail {
  emitterUpdateScript [
   &EmitterState()
   &FN_SpawnAnimTrail()
  ]
  particleSpawnScript [
   &FN_InitializeAnimTrailBoneIndecies()
   &FN_InitializeAnimTrail(AnimTrailWidth=0.2, Lifetime=0.1, SkeletalMesh=$System.SkeletalMesh)
   /SetVariables(Particles.Position=&AddVectorToPosition(Position=$Particles.Position, Vector=&Multiply_VectorByFloat(Vector=$Particles.RibbonFacing, Float=34))) {isSetParameters=true, module=SetVariables_#3}
   &AddVelocity(Velocity=$User.DebugSpeed)
  ]
  particleUpdateScript [
   &FN_UpdateAnimTrail(SkeletalMesh=$System.SkeletalMesh) {module=FN_UpdateAnimTrail001}
   &Color(
    Color=&ColorFromCurve(ColorCurve=DataInterface() {propertyValues {
       AlphaCurve {keys [<at 0 0.5 RCIM_Cubic>,<at 0.4607 0.25>,<at 1 0 RCIM_Cubic>]}
       BlueCurve {keys [<at 0 0.0156>]}
       GreenCurve {keys [<at 0 0.0156>]}
       RedCurve {keys [<at 0 0.0156>]}
      }})
    Scale Alpha=&Multiply_Float(A=1, B=$System.FadeOut)
   )
   &SolveForcesAndVelocity(Clamp Velocity=false, Limit Acceleration=false)
  ]
  rendererClasses [&NiagaraRibbonRendererProperties]
  renderers [{rendererClass=&NiagaraRibbonRendererProperties}]
 }
 NE_SoftSPrimaryTrail {
  emitterUpdateScript [
   &EmitterState()
   &FN_SpawnAnimTrail()
  ]
  particleSpawnScript [
   &FN_InitializeAnimTrailBoneIndecies(Second Bone or Socket Index=2)
   &FN_InitializeAnimTrail(Color=$User.UI_Acc_PrimaryColor, Lifetime=0.1, SkeletalMesh=$System.SkeletalMesh)
   /SetVariables(Particles.Position=&AddVectorToPosition(Position=$Particles.Position, Vector=&Multiply_VectorByFloat(Vector=$Particles.RibbonFacing, Float=-5))) {isSetParameters=true}
   &self:SpecialCaseColor(
    ColorToAnalyze:LinearColor=$User.UI_Acc_PrimaryColor
    OutputColor:LinearColor=<rgba 0 0 0 2>
    SpecialColorCase:LinearColor=<rgba 0.0203 0.0203 0.0203 1>
    Tolerance=0.1
   ) {module=SpecialCaseColor}
   &self:SetColorMagnitude_01(ColorScale=1) {module=SetColorMagnitude_01}
   &AddVelocity(Velocity=$User.DebugSpeed)
  ]
  particleUpdateScript [
   &FN_UpdateAnimTrail(SkeletalMesh=$System.SkeletalMesh) {module=FN_UpdateAnimTrail001}
   &Color(Color=&MultiplyLinearColor(
     Linear Color=&ColorFromCurve(ColorCurve=DataInterface() {propertyValues {
        AlphaCurve {keys [<at 0.3173 1>,<at 0.4763 0.2>,<at 0.5904 0 RCIM_Cubic>]}
        BlueCurve {keys [<at 0 1>]}
        GreenCurve {keys [<at 0 1>]}
        RedCurve {keys [<at 0 1>]}
       }})
     Scale Factor=&MakeVector4_FromLinearColor(LinearColor=$Particles.Initial.Color)
    ))
   /SetVariables(Particles.Color=&MultiplyLinearColor(
     Linear Color=&MultiplyLinearColor(
      Linear Color=$Particles.Color
      Scale Factor=&MakeVector4_FromLinearColor(LinearColor=&FNTimeOfDayColor(Night Color=<rgba 0.5 0.5 0.5 1>, Morning Color=<rgba 1 1 1 1>, Day Color=<rgba 1 1 1 1>, Evening Color=<rgba 1 1 1 1>))
     )
     Scale Factor=&MakeVector4(X=1, Y=1, Z=1, W=&Multiply_Float(A=0.9, B=$System.FadeOut))
    )) {isSetParameters=true}
   &SolveForcesAndVelocity(Clamp Velocity=false, Limit Acceleration=false)
  ]
  rendererClasses [&NiagaraRibbonRendererProperties]
  renderers [{rendererClass=&NiagaraRibbonRendererProperties}]
 }
 NE_Soft_SecondaryTrail {
  emitterUpdateScript [
   &EmitterState()
   &FN_SpawnAnimTrail()
  ]
  particleSpawnScript [
   &FN_InitializeAnimTrailBoneIndecies()
   &FN_InitializeAnimTrail(AnimTrailWidth=0.35, Color=$User.UI_Acc_SecondaryColor, Lifetime=0.1, SkeletalMesh=$System.SkeletalMesh)
   &self:SpecialCaseColor(
    ColorToAnalyze:LinearColor=$User.UI_Acc_SecondaryColor
    OutputColor:LinearColor=<rgba 0 0 0 1.5>
    SpecialColorCase:LinearColor=<rgba 0.0203 0.0203 0.0203 1>
    Tolerance=0.1
   ) {module=SpecialCaseColor}
   &self:SetColorMagnitude_03(ColorScale=1) {module=SetColorMagnitude_03}
   /SetVariables(Particles.Position=&AddVectorToPosition(Position=$Particles.Position, Vector=&Multiply_VectorByFloat(Vector=$Particles.RibbonFacing, Float=2))) {isSetParameters=true, module=SetVariables_#3}
   &AddVelocity(Velocity=$User.DebugSpeed)
  ]
  particleUpdateScript [
   &FN_UpdateAnimTrail(SkeletalMesh=$System.SkeletalMesh) {module=FN_UpdateAnimTrail001}
   &Color(
    Color=&MultiplyLinearColor(
     Linear Color=&ColorFromCurve(ColorCurve=DataInterface() {propertyValues {
        AlphaCurve {keys [<at 0 0.8>,<at 0.3709 0.2>,<at 0.6181 0 RCIM_Cubic>]}
        BlueCurve {keys [<at 0 1>]}
        GreenCurve {keys [<at 0 1>]}
        RedCurve {keys [<at 0 1>]}
       }})
     Scale Factor=&MakeVector4_FromLinearColor(LinearColor=$Particles.Initial.Color)
    )
    Scale Alpha=&Multiply_Float(A=0.9, B=$System.FadeOut)
   )
   &SolveForcesAndVelocity(Clamp Velocity=false, Limit Acceleration=false)
  ]
  rendererClasses [&NiagaraRibbonRendererProperties]
  renderers [{rendererClass=&NiagaraRibbonRendererProperties}]
 }
 NE_TipTrail {
  emitterUpdateScript [
   &EmitterState()
   &FN_SpawnAnimTrail(AnimTrailWidthMode=::FromSecond)
  ]
  particleSpawnScript [
   &FN_InitializeAnimTrailBoneIndecies(Second Bone or Socket Index=2)
   &FN_InitializeAnimTrail(AnimTrailWidth=0.18, Color=$User.UI_Acc_TertiaryColor, Lifetime=0.1, SkeletalMesh=$System.SkeletalMesh)
   &self:SpecialCaseColor(ColorToAnalyze:LinearColor=$User.UI_Acc_TertiaryColor, SpecialColorCase:LinearColor=<rgba 0.0203 0.0203 0.0203 1>, Tolerance=0.1) {module=SpecialCaseColor}
   &self:SetColorMagnitude(ColorScale=1) {module=SetColorMagnitude}
   &AddVelocity(Velocity=$User.DebugSpeed)
  ]
  particleUpdateScript [
   &FN_UpdateAnimTrail(SkeletalMesh=$System.SkeletalMesh) {module=FN_UpdateAnimTrail001}
   &Color(Color=&MultiplyLinearColor(
     Linear Color=&ColorFromCurve(ColorCurve=DataInterface() {propertyValues {AlphaCurve {keys [<at 0 1>,<at 0.5088 0 RCIM_Cubic>]}, BlueCurve {keys [<at 0 2>]}, GreenCurve {keys [<at 0 2>]}, RedCurve {keys [<at 0 2>]}}})
     Scale Factor=&MakeVector4_FromLinearColor(LinearColor=$Particles.Initial.Color)
    ))
   /SetVariables(Particles.Color=&MultiplyLinearColor(
     Linear Color=&MultiplyLinearColor(
      Linear Color=$Particles.Color
      Scale Factor=&MakeVector4_FromLinearColor(LinearColor=&FNTimeOfDayColor(Night Color=<rgba 0.2 0.2 0.2 1>, Morning Color=<rgba 1 1 1 1>, Day Color=<rgba 1 1 1 1>, Evening Color=<rgba 1 1 1 1>))
     )
     Scale Factor=&MakeVector4(X=1, Y=1, Z=1, W=&Multiply_Float(A=1, B=$System.FadeOut))
    )) {isSetParameters=true}
   &SolveForcesAndVelocity(Clamp Velocity=false, Limit Acceleration=false)
  ]
  rendererClasses [&NiagaraRibbonRendererProperties]
  renderers [{rendererClass=&NiagaraRibbonRendererProperties}]
 }
}
summary {
 systemName=NS_Pickaxe_UpbeatIguana_AnimTrails
 emitters [
  {emitterName=NE_SoftDarkTrail, rendererClasses [&NiagaraRibbonRendererProperties]}
  {emitterName=NE_SoftSPrimaryTrail, rendererClasses [&NiagaraRibbonRendererProperties]}
  {emitterName=NE_TipTrail, rendererClasses [&NiagaraRibbonRendererProperties]}
  {emitterName=NE_Soft_SecondaryTrail, rendererClasses [&NiagaraRibbonRendererProperties]}
 ]
 userVariables [
  User.DebugSpeed {type=Vector3f, description="", defaultValue {struct=Vector3f, value=<0 0 0>}}
  User.RibbonIDOffset {type=NiagaraInt32}
  User.UI_Acc_PrimaryColor {type=LinearColor, description="", defaultValue {struct=LinearColor, value=<rgba 0.0612 0.5711 0.0513 1>}}
  User.UI_Acc_SecondaryColor {type=LinearColor, description="", defaultValue {struct=LinearColor, value=<rgba 0.5841 0 0.1329 1>}}
  User.UI_Acc_TertiaryColor {type=LinearColor, description="", defaultValue {struct=LinearColor, value=<rgba 0.9823 0.4452 0 1>}}
 ]
}
