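# Physics asset for the UpbeatIguana pickaxe, a two-handed melee weapon cosmetic.
# It defines a single capsule body on the Melee_TwoHanded_Handle bone that serves as the simple collision shape for the axe mesh.
# There are no constraints, so the weapon acts as one rigid piece with no articulated or simulated sub-parts.
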PhysicsAsset /BRCosmetics/Weapons/FORT_Melee/Pickaxe_UpbeatIguana/Meshes/UpbeatIguana_Axe_PhysicsAsset
bodies {Melee_TwoHanded_Handle [{shapeType=Capsule, length=64.7763, radius=32.9749, center=<0.9341 7.7484 25.7449>, rotation=<rot 0.186 -90 0.0038>}]}
constraints []
